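# Physics asset for the M_MED_BirdNest male player body cosmetic, providing secondary motion for its jacket.
# Capsules on the pelvis and thighs form the body volumes, with the thighs and three small jacket base bodies fully swing- and twist-locked to the pelvis.
# Three box bodies for the jacket panels hang from those bases on constraints with tight swing limits and a zero twist limit.
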PhysicsAsset /BRCosmetics/Characters/Player/Male/Medium/Bodies/M_MED_BirdNest/Meshes/M_MED_BirdNest_Physics
bodies {
 dyn_jacket_b_l_base [{shapeType=Capsule, length=0.505, radius=0.505, shapeName=dyn_jacket_b_l_base_capsule, rotation=<rot 0 -90 0>}]
 dyn_jacket_b_m_base [{shapeType=Capsule, length=0.505, radius=0.505, shapeName=dyn_jacket_b_m_base_capsule, rotation=<rot 0 -90 0>}]
 dyn_jacket_b_r_base [{shapeType=Capsule, length=0.505, radius=0.505, shapeName=dyn_jacket_b_r_base_capsule, rotation=<rot 0 -90 0>}]
 dyn_jacket_bl_01 [{shapeType=Box, extentX=12.3536, extentY=1.5, extentZ=8.3474, shapeName=dyn_jacket_bl_01_box, center=<-3.4815 0 0>}]
 dyn_jacket_br_01 [{shapeType=Box, extentX=12.3536, extentY=1.5, extentZ=8.3474, shapeName=dyn_jacket_br_01_box, center=<-3.4815 0 0>}]
 dyn_jacket_sr_01 [{shapeType=Box, extentX=12.3536, extentY=1.5, extentZ=16.1423, shapeName=dyn_jacket_sr_01_box, center=<-3.4815 0 0>}]
 pelvis [{shapeType=Capsule, length=10.3191, radius=10, shapeName=pelvis_capsule, center=<0.2048 -1.0008 0>, rotation=<rot -2.3454 -173.0546 -180>}]
 thigh_l [{shapeType=Capsule, length=11, radius=10, shapeName=thigh_l_capsule, center=<-6.4789 -1.0885 -0.2983>, rotation=<rot 90 180 -177.0682>}]
 thigh_r [{shapeType=Capsule, length=11, radius=10, shapeName=thigh_r_capsule, center=<6.4405 1.4075 0>, rotation=<rot 90 2.3859 11.2044>}]
}
constraints [
 {
  bone1Name=thigh_l
  bone2Name=pelvis
  swing1LimitDegrees=45
  swing1Motion=Locked
  swing2LimitDegrees=45
  swing2Motion=Locked
  twistLimitDegrees=45
  twistMotion=Locked
 }
 {
  bone1Name=thigh_r
  bone2Name=pelvis
  swing1LimitDegrees=45
  swing1Motion=Locked
  swing2LimitDegrees=45
  swing2Motion=Locked
  twistLimitDegrees=45
  twistMotion=Locked
 }
 {bone1Name=dyn_jacket_bl_01, bone2Name=dyn_jacket_b_l_base, swing1LimitDegrees=15, swing2LimitDegrees=5, twistLimitDegrees=0}
 {bone1Name=dyn_jacket_sr_01, bone2Name=dyn_jacket_b_m_base, swing1LimitDegrees=15, swing2LimitDegrees=5, twistLimitDegrees=0}
 {bone1Name=dyn_jacket_br_01, bone2Name=dyn_jacket_b_r_base, swing1LimitDegrees=15, swing2LimitDegrees=5, twistLimitDegrees=0}
 {
  bone1Name=dyn_jacket_b_l_base
  bone2Name=pelvis
  swing1LimitDegrees=45
  swing1Motion=Locked
  swing2LimitDegrees=45
  swing2Motion=Locked
  twistLimitDegrees=45
  twistMotion=Locked
 }
 {
  bone1Name=dyn_jacket_b_m_base
  bone2Name=pelvis
  swing1LimitDegrees=45
  swing1Motion=Locked
  swing2LimitDegrees=45
  swing2Motion=Locked
  twistLimitDegrees=45
  twistMotion=Locked
 }
 {
  bone1Name=dyn_jacket_b_r_base
  bone2Name=pelvis
  swing1LimitDegrees=45
  swing1Motion=Locked
  swing2LimitDegrees=45
  swing2Motion=Locked
  twistLimitDegrees=45
  twistMotion=Locked
 }
]
