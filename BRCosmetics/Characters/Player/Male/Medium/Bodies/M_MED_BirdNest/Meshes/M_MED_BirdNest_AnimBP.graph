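# Post-process style cosmetic AnimBlueprint for the BirdNest medium male body, adding secondary cloth motion to the parent mesh's copied pose.
# Pose runs through deform and secondary Control Rigs, pelvis-space RigidBody, then AnimDynamics springs for collar and cuffs.
# Each update feeds RigidBody a state-based gravity override using a deprecated helper.
# It also drives base and overlay material WPO cloth from calf motion, enabling wind and turbulence during skydiving or parachuting.
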AnimBlueprint /BRCosmetics/Characters/Player/Male/Medium/Bodies/M_MED_BirdNest/Meshes/M_MED_BirdNest_AnimBP
target_skeleton=&/M_MED_BirdNest_Skeleton
graphs [
 AnimGraph {nodes [
   n1 = Control Rig(Source=Copy Pose From Mesh() {bUseAttachedParent=true}) {
    alphaBoolBlend {blendOption=Linear}
    controlRigAssetReference {blueprintRigClass=&/M_MED_BirdNest_Deform_CtrlRig.M_MED_BirdNest_Deform_CtrlRig_C}
   }
   n2 = Control Rig(Source=@n1) {
    alphaBoolBlend {blendOption=Linear}
    controlRigAssetReference {blueprintRigClass=&/M_MED_BirdNest_CtrlRig.M_MED_BirdNest_CtrlRig_C}
   }
   n3 = RigidBody!(ComponentPose=Local To Component(LocalPose=@n2), Alpha, OverrideWorldGravity:Vector=Get Gravity Override(), ExternalForce) -> (Pose) {
    bClampLinearTranslationLimitToRefPose=true
    bOverrideWorldGravity=true
    overridePhysicsAsset=&/M_MED_BirdNest_Physics
    simulationSpace=BaseBoneSpace
    alphaBoolBlend {blendOption=Linear}
    baseBoneRef {boneName=pelvis}
    componentAppliedLinearAccClamp=<4000 4000 4000>
    componentLinearAccScale=<0.3 0.3 0.3>
    componentLinearVelScale=<0.1 0.1 0.1>
    simSpaceSettings {
     maxAngularAcceleration=250
     maxAngularVelocity=3
     maxLinearAcceleration=400
     maxLinearVelocity=250
     velocityScaleZ=0
     worldAlpha=1
     externalLinearDragV=<0.1 0.1>
    }
   }
   n4 = Anim Dynamics - Bone: dyn_collar_back_01(ComponentPose=@n3, Alpha=0.7, ExternalForce="0,0,0") {
    angularDampingOverride=0.9995
    angularSpringConstant=2000
    bAngularSpring=true
    bGravityOverrideInSimSpace=true
    bLinearSpring=true
    bOverrideAngularDamping=true
    bOverrideLinearDamping=true
    bUseGravityOverride=true
    linearDampingOverride=0.9995
    linearSpringConstant=50
    simulationSpace=RootRelative
    alphaBoolBlend {blendOption=Linear}
    boundBone {boneName=dyn_collar_back_01}
    componentAppliedLinearAccClamp=<1000 1000 1000>
    gravityOverride {z=-600}
    physicsBodyDefinitions [{
      boundBone {boneName=dyn_collar_back_01}
      boxExtents=<10 10 10>
      constraintSetup {angularLimitsMax=<0 0 20>, angularLimitsMin=<0 0 -20>, angularTarget=<1 0 0>, linearAxesMax=<4 0 4>, linearAxesMin=<-3 0 -3>}
      localJointOffset=<0 0 0>
     }]
   }
   n5 = Anim Dynamics - Bone: dyn_arm_cuff_l(ComponentPose=@n4, Alpha=0.9, ExternalForce="0,0,0") {
    angularDampingOverride=1
    angularSpringConstant=3000
    bAngularSpring=true
    bOverrideAngularDamping=true
    simulationSpace=RootRelative
    alphaBoolBlend {blendOption=Linear}
    alphaScaleBiasClamp {interpSpeedDecreasing=500, interpSpeedIncreasing=100}
    boundBone {boneName=dyn_arm_cuff_l}
    componentAppliedLinearAccClamp=<1000.01 1000 1000>
    physicsBodyDefinitions [{
      boundBone {boneName=dyn_arm_cuff_l}
      boxExtents=<10 10 10>
      constraintSetup {angularLimitsMax=<0 5 5>, angularLimitsMin=<0 -5 -5>, angularTarget=<1 0 0>}
      localJointOffset=<0 0 0>
     }]
    relativeSpaceBone {boneName=lowerarm_twist_02_l}
   }
   n6 = Anim Dynamics - Bone: dyn_arm_cuff_r(ComponentPose=@n5, Alpha=0.9, ExternalForce="0,0,0") {
    angularDampingOverride=1
    angularSpringConstant=3000
    bAngularSpring=true
    bOverrideAngularDamping=true
    simulationSpace=RootRelative
    alphaBoolBlend {blendOption=Linear}
    alphaScaleBiasClamp {interpSpeedDecreasing=500, interpSpeedIncreasing=100}
    boundBone {boneName=dyn_arm_cuff_r}
    componentAppliedLinearAccClamp=<1000.01 1000 1000>
    physicsBodyDefinitions [{
      boundBone {boneName=dyn_arm_cuff_r}
      boxExtents=<10 10 10>
      constraintSetup {angularLimitsMax=<0 5 5>, angularLimitsMin=<0 -5 -5>, angularTarget=<1 0 0>}
      localJointOffset=<0 0 0>
     }]
    relativeSpaceBone {boneName=lowerarm_twist_02_r}
   }
   n7 = Anim Dynamics - Bone: dyn_pant_cuff_l(ComponentPose=@n6, Alpha=0.8, ExternalForce="0,0,0") {
    angularDampingOverride=1
    angularSpringConstant=1300
    bAngularSpring=true
    bOverrideAngularDamping=true
    simulationSpace=RootRelative
    alphaBoolBlend {blendOption=Linear}
    alphaScaleBiasClamp {interpSpeedDecreasing=500, interpSpeedIncreasing=100}
    boundBone {boneName=dyn_pant_cuff_l}
    componentAppliedLinearAccClamp=<1000.01 1000 1000>
    physicsBodyDefinitions [{
      boundBone {boneName=dyn_pant_cuff_l}
      boxExtents=<10 10 10>
      constraintSetup {angularLimitsMax=<0 10 10>, angularLimitsMin=<0 -10 -10>, angularTarget=<1 0 0>}
      localJointOffset=<0 0 0>
     }]
    relativeSpaceBone {boneName=calf_twist_02_l}
   }
   n8 = Anim Dynamics - Bone: dyn_pant_cuff_r(ComponentPose=@n7, Alpha=0.8, ExternalForce="0,0,0") {
    angularDampingOverride=1
    angularSpringConstant=1300
    bAngularSpring=true
    bOverrideAngularDamping=true
    simulationSpace=RootRelative
    alphaBoolBlend {blendOption=Linear}
    alphaScaleBiasClamp {interpSpeedDecreasing=500, interpSpeedIncreasing=100}
    boundBone {boneName=dyn_pant_cuff_r}
    componentAppliedLinearAccClamp=<1000.01 1000 1000>
    physicsBodyDefinitions [{
      boundBone {boneName=dyn_pant_cuff_r}
      boxExtents=<10 10 10>
      constraintSetup {angularLimitsMax=<0 10 10>, angularLimitsMin=<0 -10 -10>, angularTarget=<1 0 0>}
      localJointOffset=<0 0 0>
     }]
    relativeSpaceBone {boneName=calf_twist_02_r}
   }
   Output Pose(Result=Component To Local(ComponentPose=@n8)) {layerGroup=None}
  ]}
 EventGraph {nodes [
   n1 = Event BlueprintUpdateAnimation()
   > n2 = Sequence()
   > n3 = Gravity Override State Function DEPRECATED(CharacterPartAnimInstance=Self-Reference(), GameTimeInSeconds=0, GravityOverrideParams=Get Gravity Override Params())
   > Set Gravity Override(Gravity Override=@n3)
   n4 = Get Mesh!() -> (Mesh)
   n5 = Get Overlay Mat Instance Dyanmic() -> (Overlay Mat Instance Dyanmic:Material Instance Dynamic Object Reference)
   n6 = Get Material Instance Dynamic() -> (Material Instance Dynamic:Material Instance Dynamic Object Reference)
   > Set WPO Globals Function(
    Material=@n6
    Overlay Material=@n5
    Skeletal Mesh=@n4
    "Local Noise Transform  Location"="Get Local Noise Transform  Location"() -> ("Local Noise Transform  Location":Vector)
    Local Noise Offset=Get Local Noise Offset()
   )
   > Set WPO Regional Function!(
    execute
    Material:Material Instance Dynamic Object Reference=@n6
    Overlay Material:Material Instance Dynamic Object Reference=@n5
    Skeletal Mesh:Skeletal Mesh Component Object Reference=@n4
    Velocity Transform:Name=calf_r
    Velocity Interp Speed:double=Get Velocity Interp Speed()
    Collision Transform:Name=calf_r
    Collision Transform Axis:EAxis Enum
    Invert Collision Transform Axis:bool=false
    Delta Time X:double=@n1.DeltaTimeX
    Previous Velocity Transform Location:"Vector (by ref)"=Get R2 Velocity Transform Location()
    Previous Velocity Transform Velocity:"Vector (by ref)"=Get R2 Velocity Transform Velocity()
    Region:int=1
   ) -> (then)
   > Set WPO Regional Function(
    Material=@n6
    Overlay Material=@n5
    Skeletal Mesh=@n4
    Velocity Transform=calf_l
    Velocity Interp Speed=Get Velocity Interp Speed()
    Collision Transform=calf_l
    Invert Collision Transform Axis=true
    Delta Time X=@n1.DeltaTimeX
    Previous Velocity Transform Location=Get L Previous Velocity Transform Location() -> (L Previous Velocity Transform Location:Vector)
    Previous Velocity Transform Velocity=Get L Previous Velocity Transform Velocity() -> (L Previous Velocity Transform Velocity:Vector)
    Region=2
   )
   > SetScalarParameterValue(self=[@n6, @n5], ParameterName=ClothSim_WindStrength, Value=Get WPO Wind() -> (WPO Wind:double))
   > SetScalarParameterValue(self=[@n6, @n5], ParameterName=ClothSim_TurbulenceNoiseTileDivisor, Value=Get Turbulence Noise() -> (Turbulence Noise:float))
   n7 = Branch(
    execute=@n2.then_1
    Condition=OR Boolean!(A=Get bIsSkydiveFloating(), B=Get bIsSkydiveDiveMode(), C:bool=Get bIsParachuteOpened()) -> (ReturnValue)
   )
   > Set Velocity Interp Speed(Velocity Interp Speed:double=40) -> (Output_Get:double)
   > Set WPO Wind(WPO Wind:double=1) -> (Output_Get:double)
   > Set Turbulence Noise(Turbulence Noise:float=80) -> (Output_Get:float)
   Set Velocity Interp Speed(execute=@n7.else, Velocity Interp Speed=0)
   > Set WPO Wind(WPO Wind=0)
   > Set Turbulence Noise(Turbulence Noise=0)
   Event OnCharacterCustomizationLoaded()
   n8 = GetPartSkeletalMeshComponent()
   > n9 = Set Mesh(Mesh:Skeletal Mesh Component Object Reference=@n8) -> (Output_Get:Skeletal Mesh Component Object Reference)
   > n10 = Get Dynamic Material Instance(SkeletalMeshComponent=@n9)
   > Set Material Instance Dynamic(Material Instance Dynamic:Material Instance Dynamic Object Reference=@n10) -> (Output_Get:Material Instance Dynamic Object Reference)
   > n11 = GetOrCreateDynamicOverlayMaterial(MeshComponent=@n8, MaterialName=None)
   > Set Overlay Mat Instance Dyanmic(Overlay Mat Instance Dyanmic:Material Instance Dynamic Object Reference=@n11) -> (Output_Get:Material Instance Dynamic Object Reference)
  ]}
]
variables [
 Gravity Override {type=Vector}
 Gravity Override Params {
  type=GravityOverrideParamsStruct
  default {
   skydive_FloatingDown {gravityOverride=<0 150 1500>, windAmplitude=<0 10 1500>, windFrequency=<0 50 20>, windOffset=<0 300 0>}
   skydive_GoingUp {gravityOverride=<0 -700 1500>, windAmplitude=<200 -500 0>, windOffset=<0 -600 0>}
  }
 }
 L Previous Velocity Transform Location {type=Vector}
 L Previous Velocity Transform Velocity {type=Vector}
 Local Noise Offset {type=Vector}
 "Local Noise Transform  Location" {type=Vector}
 Material Instance Dynamic {type=MaterialInstanceDynamic}
 Mesh {type=SkeletalMeshComponent}
 Overlay Mat Instance Dyanmic {type=MaterialInstanceDynamic}
 R2 Velocity Transform Location {type=Vector}
 R2 Velocity Transform Velocity {type=Vector}
 Turbulence Noise {type=number}
 Velocity Interp Speed {type=number}
 WPO Wind {type=number}
]
